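# Niagara effect for the DangerIndex backpack cosmetic: a bubble sprite that inflates with a wobble and pops in a short burst.
# Timeline curves and a rewind state drive bubble scale, while audio reactivity also adjusts size, tiling and edge strength.
# Bubble alpha follows time of day, and camera offset follows view angle.
# A shared TriggerBurst flag spawns two splashes, a corona, pixie dust and a refraction mesh, sized by GlobalScale.
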NiagaraSystem /BRCosmetics/Accessories/FORT_Backpacks/Backpack_F_MED_DangerIndex/FX/NS_Backpack_DangerIndex_Bubble
dependencies {
 usedDataInterfaces [
  &NiagaraDataInterfaceCurve
  &NiagaraDataInterfaceVector2DCurve
  &NiagaraDataInterfaceSpriteRendererInfo
  &NiagaraDataInterfaceColorCurve
  &NiagaraDataInterfaceCamera
 ]
 usedDynamicInputs [
  &MaskVector2dBySpawnGroup
  /Niagara/DynamicInputs/Lerp/SmoothLerpOverTime_Float
  &RemapRange
  &Waveform
  &Boolean_AndOperation
  &Abs_Float
  &DotProduct
  &ReturnCameraVector
 ]
 usedModules [
  &self:SystemSpawnScript.SystemScriptSource.SystemScriptGraph.NiagaraNodeAssignment_4.SetVariables_#1
  &self:SystemSpawnScript.SystemScriptSource.SystemScriptGraph.NiagaraNodeAssignment_2.SetVariables_#2
  &self:SystemSpawnScript.SystemScriptSource.SystemScriptGraph.NiagaraNodeAssignment_3.SetVariables_#3
  &SystemState
  /Game/Effects/Niagara/Modules/Audio/FN_FakeAudioReactivityLoop
  &self:SystemSpawnScript.SystemScriptSource.SystemScriptGraph.NiagaraNodeAssignment_1.SetVariables_#4
  &self:SystemSpawnScript.SystemScriptSource.SystemScriptGraph.NiagaraNodeAssignment_0.SetVariables_#5
  &Timeline
  &self:SystemSpawnScript.SystemScriptSource.SystemScriptGraph.NiagaraNodeAssignment_5.SetVariables_#6
  &self:SystemSpawnScript.SystemScriptSource.SystemScriptGraph.NiagaraNodeAssignment_6.SetVariables_#7
 ]
}
emitters {
 Bubble {
  emitterUpdateScript [
   &EmitterState(Life Cycle Mode=::NewEnumerator1, Loop Behavior=::NewEnumerator1, Loop Duration Mode=::NewEnumerator1, UseLoopDelay=false)
   &SpawnBurst_Instantaneous(Use Spawn Probability=false, UseLoopCountLimit=false)
   /SetVariables(
    Emitter.BubbleTiling:NiagaraFloat=&Lerp_Float(
     A=0.2
     B=0.1
     Alpha=&FloatFromCurve(
      FloatCurve=DataInterface() {propertyValues {Curve {
         defaultValue=1
         keys [
          {arriveTangent=-0.5113, interpMode=RCIM_Cubic, time=0.3, value=0}
          {arriveTangent=1.6395, interpMode=RCIM_Cubic, leaveTangent=1.6395, tangentMode=RCTM_User, time=1, value=1}
         ]
        }}}
      CurveIndex=$System.AudioReactivity
     )
    )
    Emitter.EdgeStrength:NiagaraFloat=&Lerp_Float(
     A=0.5
     B=0.3
     Alpha=&FloatFromCurve(
      FloatCurve=DataInterface() {propertyValues {Curve {
         defaultValue=1
         keys [
          {arriveTangent=-0.5113, interpMode=RCIM_Cubic, time=0, value=0}
          {arriveTangent=1.6395, interpMode=RCIM_Cubic, leaveTangent=1.6395, tangentMode=RCTM_User, time=1, value=1}
         ]
        }}}
      CurveIndex=$System.AudioReactivity
     )
    )
   ) {isSetParameters=true}
  ]
  particleSpawnScript [&InitializeParticle(
    Color=<rgba 0.7 0.85 1 2>
    Lifetime=1.5
    Position Offset=<0 -10 5>
    Sprite Size=<35 45>
    Sprite Size Mode=::NewEnumerator4
    UsePositionOffset=true
    Write Material Random=true
   )]
  particleUpdateScript [
   &ParticleState(Kill Particles When Lifetime Has Elapsed=false)
   &ScaleSpriteSize(
    Scale Factor=&Multiply_Vector2D_ByFloat(
     Float=&FN_FrontEndFloatModifier(
      Gameplay=1
      Front End=&MakeCustomFloatFromBool(Bool=&SetBoolByFloatComparison(A=$System.Timeline.NormalizedTime, B=0.01), True Float=1, False Float=0)
     )
     Vector2D=/Niagara/DynamicInputs/Add/Add_Vector2(
      A:Vector2f=&Add_Vector2(
       A:Vector2f=&MakeCustomVector2DFromBool(
        Bool=$System.Timeline.IsRewinding
        True Vector2D=&Vector2DFromCurve(
         Vector2Curve=DataInterface() {propertyValues {XCurve {keys [<at 0 1 RCIM_Constant>,<at 0.99 0 RCIM_Constant>]}, YCurve {keys [<at 0 1 RCIM_Constant>,<at 0.99 0 RCIM_Constant>]}}}
         CurveIndex=$System.Timeline.InverseNormalizedTime
         ScaleCurve=<1 1>
        )
        False Vector2D=&Vector2DFromCurve(
         Vector2Curve=DataInterface() {propertyValues {
           XCurve {keys [
             {arriveTangent=0.3581, interpMode=RCIM_Cubic, leaveTangent=0.3581, tangentMode=RCTM_User, time=0.0039, value=0}
             {arriveTangent=2.3597, interpMode=RCIM_Cubic, leaveTangent=2.3597, time=0.2203, value=0.3954}
             {arriveTangent=-0.2652, interpMode=RCIM_Cubic, leaveTangent=-0.2652, tangentMode=RCTM_User, time=0.5124, value=1.2}
             {arriveTangent=-0.2843, interpMode=RCIM_Cubic, leaveTangent=-0.2843, time=0.728, value=0.5851}
             {arriveTangent=0.3265, interpMode=RCIM_Cubic, leaveTangent=0.3265, time=1.0095, value=1.0587}
             {arriveTangent=0.0736, interpMode=RCIM_Cubic, leaveTangent=0.0736, tangentMode=RCTM_User, time=1.4073, value=0.8069}
             {arriveTangent=0.3985, interpMode=RCIM_Cubic, leaveTangent=0.3985, tangentMode=RCTM_User, time=1.6591, value=0.9889}
            ]}
           YCurve {keys [
             {arriveTangent=2.6784, interpMode=RCIM_Cubic, leaveTangent=2.6784, tangentMode=RCTM_User, time=0, value=0}
             {arriveTangent=1.1878, interpMode=RCIM_Cubic, leaveTangent=1.1878, time=0.298, value=1.0906}
             {arriveTangent=-0.2075, interpMode=RCIM_Cubic, leaveTangent=-0.2075, time=0.6571, value=0.7805}
             {arriveTangent=0.0038, interpMode=RCIM_Cubic, leaveTangent=0.0038, time=0.7953, value=0.9874}
             {arriveTangent=0.0444, interpMode=RCIM_Cubic, leaveTangent=0.0444, time=1.0999, value=0.7822}
             {arriveTangent=0.3398, interpMode=RCIM_Cubic, leaveTangent=0.3398, time=1.4729, value=1.0175}
             {arriveTangent=0.0791, interpMode=RCIM_Cubic, leaveTangent=0.0791, tangentMode=RCTM_User, time=1.6865, value=0.9815}
            ]}
          }}
         CurveIndex=&Multiply_Float(A=$System.Timeline.Time, B=1.5)
         ScaleCurve=<1 1>
        )
       )
       B:Vector2f=&Vector2DFromFloat(Value=&Lerp_Float(
         A=0
         B=0.3
         Alpha=&FloatFromCurve(
          FloatCurve=DataInterface() {propertyValues {Curve {
             defaultValue=1
             keys [{arriveTangent=3.8414, interpMode=RCIM_Cubic, time=0.3, value=0}, {interpMode=RCIM_Cubic, leaveTangent=-0.8228, time=1, value=1}]
            }}}
          CurveIndex=$System.AudioReactivity
         )
        ))
      )
      B:Vector2f=&Vector2DFromFloat(Value=&Multiply_Float(
        A=&MakeCustomFloatFromBool(Bool=$System.Timeline.IsRewinding, True Float=$System.Timeline.InverseNormalizedTime, False Float=0)
        B=0.4
       ))
     )
    )
    Scale Sprite Size Mode=::NewEnumerator2
   )
   &SolveForcesAndVelocity(Clamp Velocity=false, Limit Acceleration=false)
   &ScaleColor(
    Scale Alpha=&FNTimeOfDayScalar(Night Scale=0.3, Morning Scale=1, Evening Scale=0.3)
    ScaleA=true
    ScaleRGB=true
   )
   &CameraOffset(Camera Offset Amount=&Lerp_Float(
     A=12
     B=25
     Alpha=&FloatFromCurve(FloatCurve=DataInterface() {propertyValues {Curve {keys [<at 0.98 0>,<at 1 1>]}}}, CurveIndex=$System.CameraAngle)
    ))
   /SetVariables(Particles.RandomScaleFloat:NiagaraFloat=&MakeFloatFromVector2d(Vector 2D=$Particles.Initial.SpriteSize)) {enabled=false, isSetParameters=true}
  ]
  rendererClasses [&NiagaraSpriteRendererProperties]
  renderers [{rendererClass=&NiagaraSpriteRendererProperties}]
 }
 Corona {
  emitterUpdateScript [
   &EmitterState(Life Cycle Mode=::NewEnumerator1, Loop Duration=1, UseLoopDelay=false)
   /Niagara/Modules/Emitter/SpawnByTrigger(Trigger:NiagaraBool=$System.TriggerBurst, Trigger Count=1, Trigger Interval=0.01, TriggerCountSwitch=true, TriggerIntervalSwitch=true)
  ]
  particleSpawnScript [
   &InitializeParticle(
    Lifetime=0.5
    Position Offset=&Add_Vector(A=<0 0 0>, B=$System.Offset)
    Sprite Size Mode=::NewEnumerator3
    Uniform Sprite Size=&Multiply_Float(A=$User.GlobalScale, B=10)
    UsePositionOffset=true
    Write Material Random=true
   )
   &ShapeLocation(MaskSpawnGroup=false, Override Seed=false, Sphere Radius=1, UseOwnerScale=true)
  ]
  particleUpdateScript [
   &ParticleState()
   /SetVariables(
    Particles.CoronaColor:LinearColor=&MultiplyLinearColor(
     Linear Color=<rgba 2 2 2 1>
     Scale Factor=&MakeVector4_FromLinearColor(LinearColor=&HueShiftLinearColor(
       Hue=0
       Saturation=0.5
       Value=1
       Linear Color=&ColorFromCurve(ColorCurve=DataInterface() {propertyValues {
          AlphaCurve {keys [<at 0.0094 0.4>, {arriveTangent=-0.4106, interpMode=RCIM_Cubic, leaveTangent=-0.4106, time=0.233, value=0.2}, <at 0.9836 0>]}
          BlueCurve {keys [<at 0.4512 0>,<at 0.9742 1>]}
          GreenCurve {keys [<at 0.4512 0.7321>,<at 0.9742 0.3437>]}
          RedCurve {keys [<at 0.4512 1>,<at 0.9742 0.1562>]}
         }})
      ))
    )
    Particles.CoronaDyn:Vector4f=&MakeVector4(
     X=&FloatFromCurve(FloatCurve=DataInterface() {propertyValues {Curve {keys [<at 0 1>,<at 1 0.8>]}}})
     Y=0
     Z=1
     W=0
    )
    Particles.CoronaSize:Vector2f=&Vector2DFromFloat(Value=&Multiply_Float(
      A=&FloatFromCurve(
       FloatCurve=DataInterface() {propertyValues {Curve {keys [{arriveTangent=0.3582, interpMode=RCIM_Cubic, time=0, value=1}, <at 1 1 RCIM_Cubic>]}}}
       Scale Curve=&Multiply_Float(A=$User.GlobalScale, B=85)
      )
      B=1
     ))
   ) {isSetParameters=true}
  ]
  rendererClasses [&NiagaraSpriteRendererProperties]
  renderers [{rendererClass=&NiagaraSpriteRendererProperties}]
 }
 PixieDust {
  emitterUpdateScript [
   &EmitterState(Inactive Response=::NewEnumerator1, Life Cycle Mode=::NewEnumerator1, UseLoopDelay=false)
   &SpawnByTrigger(Trigger:NiagaraBool=$System.TriggerBurst, Trigger Count=1, Trigger Interval=0.01, TriggerCountSwitch=true, TriggerIntervalSwitch=true)
  ]
  particleSpawnScript [
   &FN_SetSpawnGroup(Number of Spawn Groups=2)
   &InitializeParticle(
    Color=<rgba 0.9445 0.9445 0.9445 0.5>
    Lifetime Max=0.1
    Lifetime Min=0.1
    Lifetime Mode=::NewEnumerator1
    Position Offset=$System.Offset
    Position Offset Coordinate Space=::Local
    Sprite Rotation Mode=::NewEnumerator2
    Sprite Size Mode=::NewEnumerator3
    Uniform Sprite Size=&Multiply_Float(A=80, B=$User.GlobalScale)
    UsePositionOffset=true
    Write Material Random=true
   )
   &DynamicMaterialParameters(
    Index 0 Param 1=&RandomRangeFloat(Minimum=-0.025, Maximum=0.025)
    Index 0 Param 2=&RandomRangeFloat(Minimum=0, Maximum=32)
    Module.Index0Param0Write=true
    Module.Index0Param1Write=true
    Module.Index0Param2Write=false
    Module.Index0Param3Write=false
   )
   &CameraOffset(Camera Offset Amount=&MergeHelper_FloatSwitch(0=10, 1=10, 2=25, 3=25, 4=25, 5=25, 6=25, 7=25, 8=25, 9=25, Index=0, Num=::NewEnumerator8))
  ]
  particleUpdateScript [
   &ParticleState()
   &ScaleSpriteSize(Uniform Curve Sprite Scale=DataInterface() {propertyValues {Curve {keys [{arriveTangent=0.516, interpMode=RCIM_Cubic, leaveTangent=0.516, tangentMode=RCTM_User, time=0, value=0.8}, <at 1 1 RCIM_Cubic>]}}})
   &ScaleColor(
    Scale Alpha=&FloatFromCurve(FloatCurve=DataInterface() {propertyValues {Curve {defaultValue=0, keys [{interpMode=RCIM_Cubic, tangentMode=RCTM_User, time=0, value=1}, {tangentMode=RCTM_User, time=1, value=0}]}}})
    Scale RGB=<1.1 1.1 1.1>
    ScaleA=true
    ScaleRGB=true
   )
   &SolveForcesAndVelocity(Clamp Velocity=false, Limit Acceleration=false)
  ]
  rendererClasses [&NiagaraSpriteRendererProperties]
  renderers [{rendererClass=&NiagaraSpriteRendererProperties}]
 }
 Refraction {
  emitterUpdateScript [
   &EmitterState(Life Cycle Mode=::NewEnumerator1, Loop Delay=0.075, Loop Duration=2, UseLoopDelay=true)
   &SpawnByTrigger(Trigger:NiagaraBool=$System.TriggerBurst, Trigger Count=1, Trigger Interval=0.01, TriggerCountSwitch=true, TriggerIntervalSwitch=true)
  ]
  particleSpawnScript [&InitializeParticle(
    Color=<rgba 1 1 1 0.5>
    Lifetime=0.15
    Mesh Scale Mode=::NewEnumerator3
    Mesh Uniform Scale=&Multiply_Float(A=0.25, B=$User.GlobalScale)
    Position Offset=$System.Offset
    Sprite Size Mode=::NewEnumerator3
    Uniform Sprite Size=&Multiply_Float(A=70, B=$User.GlobalScale)
    UsePositionOffset=true
    Write Material Random=true
   )]
  particleUpdateScript [
   &ParticleState()
   &ScaleSpriteSize(
    Scale Sprite Size Mode=::NewEnumerator0
    Uniform Scale Factor=&FloatFromCurve(FloatCurve=DataInterface() {propertyValues {Curve {keys [{arriveTangent=0.513, interpMode=RCIM_Cubic, leaveTangent=0.513, tangentMode=RCTM_User, time=0, value=0.8}, <at 1 1 RCIM_Cubic>]}}})
   )
   &ScaleMeshSize(Scale Factor=&VectorFromFloat(Value=&FloatFromCurve(FloatCurve=DataInterface() {propertyValues {Curve {keys [{arriveTangent=0.3963, interpMode=RCIM_Cubic, leaveTangent=0.3963, tangentMode=RCTM_User, time=0, value=0.7}, <at 1 1 RCIM_Cubic>]}}})))
   &DynamicMaterialParameters(
    Index 0 Param 1=&MaskFloatBySpawnGroup(
     Masked Float=&FloatFromCurve(FloatCurve=DataInterface() {propertyValues {Curve {keys [{arriveTangent=0.1648, interpMode=RCIM_Cubic, leaveTangent=0.1648, tangentMode=RCTM_User, time=0, value=0.9}, <at 0.5 1 RCIM_Cubic>]}}})
     Passthrough Float=&FloatFromCurve(FloatCurve=DataInterface() {propertyValues {Curve {keys [<at 0 0.9 RCIM_Cubic>,<at 1 0 RCIM_Cubic>]}}}, Scale Curve=2)
    )
    Index 0 Param 3=5
    Index 0 Param 4=0.04
    Module.Index0Float0Write=true
    Module.Index0Float1Write=false
    Module.Index0Float2Write=true
    Module.Index0Float3Write=true
   )
   &ScaleColor(
    Scale Alpha=&FloatFromCurve(FloatCurve=DataInterface() {propertyValues {Curve {keys [<at 0 1>,<at 1 0>]}}})
    ScaleA=true
    ScaleRGB=true
   )
  ]
  rendererClasses [&NiagaraMeshRendererProperties]
  renderers [{rendererClass=&NiagaraMeshRendererProperties, rendererIndex=0}]
 }
 Splash_01 {
  emitterUpdateScript [
   &EmitterState(Life Cycle Mode=::NewEnumerator1, Loop Duration=1, UseLoopDelay=false)
   &SpawnByTrigger(Trigger:NiagaraBool=$System.TriggerBurst, Trigger Count=1, Trigger Interval=0.01, TriggerCountSwitch=true, TriggerIntervalSwitch=true)
  ]
  particleSpawnScript [
   &InitializeParticle(
    Color=<rgba 0.6042 0.8417 1 0.3>
    Lifetime Max=0.15
    Lifetime Min=0.1
    Lifetime Mode=::NewEnumerator1
    Position Offset=$System.Offset
    Sprite Rotation Mode=::NewEnumerator2
    Sprite Size Mode=::NewEnumerator3
    Uniform Sprite Size=&Multiply_Float(A=$User.GlobalScale, B=&RandomRangeFloat(Minimum=40, Maximum=50))
    UsePositionOffset=true
    Write Material Random=true
   ) {module=InitializeParticle002}
   /SetVariables(Particles.SubImageIndex=&RandomRangeFloat(Minimum=0, Maximum=3)) {isSetParameters=true}
  ]
  particleUpdateScript [
   &ParticleState()
   &AccelerationForce(Acceleration=&UniformRangedVector(Minimum=<0 0 -50>, Maximum=<0 0 -80>))
   &SolveForcesAndVelocity(Clamp Velocity=false, Limit Acceleration=false)
   &ScaleSpriteSize(Scale Factor=&Vector2DFromFloat(Value=&FloatFromCurve(
      FloatCurve=DataInterface() {propertyValues {Curve {keys [
          {arriveTangent=8.9659, interpMode=RCIM_Cubic, leaveTangent=8.9659, tangentMode=RCTM_User, time=0, value=0}
          {arriveTangent=0.002, interpMode=RCIM_Cubic, leaveTangent=0.6693, tangentMode=RCTM_Break, time=0.1, value=0.9069}
          {arriveTangent=0.4183, interpMode=RCIM_Cubic, leaveTangent=0.4183, tangentMode=RCTM_User, time=1, value=1.2}
         ]}}}
      Scale Curve=1.2
     )))
   &ScaleColor(
    Scale Alpha=&FloatFromCurve(FloatCurve=DataInterface() {propertyValues {Curve {keys [<at 0 1>,<at 1 0>]}}})
    ScaleA=true
    ScaleRGB=true
   )
  ]
  rendererClasses [&NiagaraSpriteRendererProperties]
  renderers [{rendererClass=&NiagaraSpriteRendererProperties}]
 }
 Splash_02 {
  emitterUpdateScript [
   &EmitterState(Life Cycle Mode=::NewEnumerator1, Loop Duration=1, UseLoopDelay=false)
   &SpawnByTrigger(Trigger:NiagaraBool=$System.TriggerBurst, Trigger Count=1, Trigger Interval=0.01, TriggerCountSwitch=true, TriggerIntervalSwitch=true)
  ]
  particleSpawnScript [
   &InitializeParticle(
    Color=<rgba 1 1 1 0.2>
    Lifetime Max=0.12
    Lifetime Min=0.12
    Lifetime Mode=::NewEnumerator1
    Position Offset=$System.Offset
    Sprite Rotation Mode=::NewEnumerator2
    Sprite Size Mode=::NewEnumerator3
    Uniform Sprite Size=&Multiply_Float(A=$User.GlobalScale, B=&RandomRangeFloat(Minimum=40, Maximum=50))
    UsePositionOffset=true
    Write Material Random=true
   )
   &SubUVAnimation(
    End Frame Range Override=3
    Evaluation Type=::NewEnumerator0
    "Sprite Renderer "=DataInterface() {propertyValues {SpriteRenderer=&self:Splash_02_0.NiagaraSpriteRendererProperties_0}}
    SubUV Animation Mode:ENiagara_SubUVLookupModeV2=::NewEnumerator2
    UseEndFrame=true
    UseStartFrame=true
   )
   &AddVelocity()
   &CameraOffset(Camera Offset Amount=20)
  ]
  particleUpdateScript [
   &ParticleState()
   &ScaleSpriteSize(Uniform Curve Sprite Scale=DataInterface() {propertyValues {Curve {keys [{arriveTangent=0.2555, interpMode=RCIM_Cubic, leaveTangent=0.2555, tangentMode=RCTM_User, time=0, value=0.7}, <at 1 1 RCIM_Cubic>]}}})
   &GravityForce(Gravity=&RandomRangeVector(Minimum=<0 0 -500>, Maximum=<0 0 -200>))
   &SolveForcesAndVelocity(Clamp Velocity=false, Limit Acceleration=false)
   &ScaleColor(
    Scale Alpha=&FloatFromCurve(FloatCurve=DataInterface() {propertyValues {Curve {keys [<at 0 1>,<at 1 0>]}}})
    ScaleA=true
    ScaleRGB=true
   )
   &DynamicMaterialParameters(
    Index 0 Param 1=&FloatFromCurve(FloatCurve=DataInterface() {propertyValues {Curve {keys [<at 0.5 1>,<at 1 0.6>]}}})
    Module.Param0WriteEnabled=true
    Module.Param1WriteEnabled=true
    Module.Param2WriteEnabled=true
    Module.Param3WriteEnabled=true
   )
  ]
  rendererClasses [&NiagaraSpriteRendererProperties]
  renderers [{rendererClass=&NiagaraSpriteRendererProperties}]
 }
}
summary {
 systemName=NS_Backpack_DangerIndex_Bubble
 emitters [
  {emitterName=Bubble, rendererClasses [&NiagaraSpriteRendererProperties]}
  {emitterName=Splash_01, rendererClasses [&NiagaraSpriteRendererProperties]}
  {emitterName=Splash_02, rendererClasses [&NiagaraSpriteRendererProperties]}
  {emitterName=Corona, rendererClasses [&NiagaraSpriteRendererProperties]}
  {emitterName=PixieDust, rendererClasses [&NiagaraSpriteRendererProperties]}
  {emitterName=Refraction, rendererClasses [&NiagaraMeshRendererProperties]}
 ]
 userVariables [
  User.bIsFrontEnd {type=NiagaraBool}
  User.bIsFrontEndPreview {type=NiagaraBool}
  User.bIsPrimaryPreviewItem {type=NiagaraBool}
  User.EnableFrontendFakeAudioLoop {type=NiagaraFloat, defaultValue=0, description=""}
  User.GlobalScale {type=NiagaraFloat, defaultValue=0.85, description=""}
 ]
}
