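# Physics asset for the DangerIndex backpack cosmetic, defining bodies and joints for its left and right hand and thigh bones.
# Small spheres on the base bones root short chains of dynamic-bone capsules: two links per hand, one per thigh.
# Angular limits keep twist tight, with the widest swing on one thigh axis.
# The outer hand joints differ slightly, with the left allowing more secondary swing than the right.
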PhysicsAsset /BRCosmetics/Accessories/FORT_Backpacks/Backpack_F_MED_DangerIndex/Meshes/F_MED_DangerIndex_Pack_Physics
bodies {
 base_hand_l [{shapeType=Sphere, radius=1, shapeName=base_hand_l_sphere}]
 base_hand_r [{shapeType=Sphere, radius=1, shapeName=base_hand_r_sphere}]
 base_thigh_l [{shapeType=Sphere, radius=1, shapeName=base_thigh_l_sphere}]
 base_thigh_r [{shapeType=Sphere, radius=1, shapeName=base_thigh_r_sphere}]
 dyn_hand_01_l [{shapeType=Capsule, length=2, radius=1.2, shapeName=dyn_hand_01_l_capsule, center=<1 0 0>, rotation=<rot 90 0 0>}]
 dyn_hand_01_r [{shapeType=Capsule, length=2, radius=1.2, shapeName=dyn_hand_01_r_capsule, center=<1 0 0>, rotation=<rot 90 0 0>}]
 dyn_hand_02_l [{shapeType=Capsule, length=2.2, radius=1.2, shapeName=dyn_hand_02_l_capsule, center=<1 0 0>, rotation=<rot 90 0 0>}]
 dyn_hand_02_r [{shapeType=Capsule, length=2.2, radius=1.2, shapeName=dyn_hand_02_r_capsule, center=<1 0 0>, rotation=<rot 90 0 0>}]
 dyn_thigh_l [{shapeType=Capsule, length=2.3, radius=1.6, shapeName=dyn_thigh_l_capsule, center=<1.8 0 0>, rotation=<rot 90 0 0>}]
 dyn_thigh_r [{shapeType=Capsule, length=2.3, radius=1.6, shapeName=dyn_thigh_r_capsule, center=<1.8 0 0>, rotation=<rot 90 0 0>}]
}
constraints [
 {bone1Name=dyn_thigh_r, bone2Name=base_thigh_r, swing1LimitDegrees=35, swing2LimitDegrees=10, twistLimitDegrees=5}
 {bone1Name=dyn_hand_01_l, bone2Name=base_hand_l, swing1LimitDegrees=15, swing2LimitDegrees=15, twistLimitDegrees=5}
 {bone1Name=dyn_hand_02_l, bone2Name=dyn_hand_01_l, swing1LimitDegrees=10, swing2LimitDegrees=15, twistLimitDegrees=5}
 {bone1Name=dyn_hand_01_r, bone2Name=base_hand_r, swing1LimitDegrees=15, swing2LimitDegrees=15, twistLimitDegrees=5}
 {bone1Name=dyn_hand_02_r, bone2Name=dyn_hand_01_r, swing1LimitDegrees=10, swing2LimitDegrees=10, twistLimitDegrees=5}
 {bone1Name=dyn_thigh_l, bone2Name=base_thigh_l, swing1LimitDegrees=35, swing2LimitDegrees=10, twistLimitDegrees=5}
]
