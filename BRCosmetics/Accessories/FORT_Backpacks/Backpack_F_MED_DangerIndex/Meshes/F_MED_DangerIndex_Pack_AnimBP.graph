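# Animation Blueprint for the DangerIndex backpack cosmetic, adding secondary physics motion and pose-based adjustments to the pack.
# Event graph maps spine bend to pack offset, rotation and scale, and stores gravity and rigid-body settings from a parameter asset.
# In-game, RigidBody feeds additive pack transforms, including skydive offsets, then AnimDynamics for hair and earrings.
# Front-end poses instead use a separate AnimDynamics chain ending in a velocity-limited RigidBody, selected by bIsInFrontEnd.
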AnimBlueprint /BRCosmetics/Accessories/FORT_Backpacks/Backpack_F_MED_DangerIndex/Meshes/F_MED_DangerIndex_Pack_AnimBP
target_skeleton=&/F_MED_DangerIndex_Pack_Skeleton
graphs [
 AnimGraph {nodes [
   n1 = RigidBody!(
    ComponentPose=Local To Component(LocalPose=Use cached pose 'My_pose'())
    Alpha
    OverrideWorldGravity:Vector="0,0,-980"
    ExternalForce
    SimSpaceSettings:Sim Space Settings Structure=(WorldAlpha=0,VelocityScaleZ=1,DampingAlpha=1,MaxLinearVelocity=10000,MaxAngularVelocity=10000,MaxLinearAcceleration=10000,MaxAngularAcceleration=10000,ExternalLinearDragV=(X=0,Y=0,Z=0),ExternalLinearVelocity=(X=0,Y=0,Z=0),ExternalAngularVelocity=(X=0,Y=0,Z=0))
   ) -> (Pose) {
    bOverrideWorldGravity=true
    overridePhysicsAsset=&/F_MED_DangerIndex_Pack_Physics
    simulationSpace=BaseBoneSpace
    baseBoneRef {boneName=pack}
    overrideWorldGravity {z=-980}
   }
   n2 = "Transform (Modify) Bone - Bone: pack"(ComponentPose=@n1, Translation="0,1,0", Rotation="0,0,0") {rotationMode=BMM_Additive, scaleMode=BMM_Additive, translationMode=BMM_Additive, boneToModify {boneName=pack}, translation {y=1}}
   n3 = "Transform (Modify) Bone - Bone: pack"(ComponentPose=@n2, Translation="0,1,0", Rotation=Get pickaxe_rotate_r() -> (pickaxe_rotate_r:Rotator)) {rotationMode=BMM_Additive, scaleMode=BMM_Additive, translationMode=BMM_Additive, boneToModify {boneName=pack}, translation {y=1}}
   n4 = "Transform (Modify) Bone - Bone: pack"!(ComponentPose=@n3, bAlphaBoolEnabled:bool=Get bIsSkydiveDiveMode(), Translation="0,-2,-3", Rotation="0,0,5", Scale) -> (Pose) {
    alphaInputType=Bool
    rotationMode=BMM_Additive
    translationMode=BMM_Additive
    boneToModify {boneName=pack}
    rotation {roll=5}
    translation {y=-2, z=-3}
   }
   n5 = "Transform (Modify) Bone - Bone: pack"(ComponentPose=@n4, bAlphaBoolEnabled=Get bIsSkydiveFloating(), Translation="0,-2,-3", Rotation="0,0,5") {
    alphaInputType=Bool
    rotationMode=BMM_Additive
    translationMode=BMM_Additive
    boneToModify {boneName=pack}
    rotation {roll=5}
    translation {y=-2, z=-3}
   }
   n6 = Anim Dynamics - Bone: dyn_hair(
    ComponentPose="Transform (Modify) Bone - Bone: pack"!(
     ComponentPose=@n5
     Alpha:float
     Translation=Get backpack_translation()
     Rotation=Get backpack_rotate()
     Scale=Get backpack_scale() -> (backpack_scale:Vector)
    ) -> (Pose) {rotationMode=BMM_Additive, scaleMode=BMM_Additive, translationMode=BMM_Additive, boneToModify {boneName=pack}}
    Alpha=1
    ExternalForce="0,0,0"
   ) {
    bLinearSpring=true
    bOverrideLinearDamping=true
    bUseGravityOverride=true
    linearDampingOverride=1
    linearSpringConstant=300
    simulationSpace=RootRelative
    boundBone {boneName=dyn_hair}
    componentAppliedLinearAccClamp=<1000 1000 1000>
    componentLinearAccScale=<0.2 0.2 0.2>
    componentLinearVelScale=<0.3 0.3 0.3>
    gravityOverride {z=-980}
    physicsBodyDefinitions [{
      boundBone {boneName=dyn_hair}
      boxExtents=<50 50 50>
      constraintSetup {angularLimitsMax=<0 0 0>, angularLimitsMin=<0 0 0>, angularTarget=<1 0 0>, linearAxesMax=<10 0 0>}
      localJointOffset=<0.3 0 0>
     }]
   }
   n7 = Anim Dynamics - Bone: dyn_earring_l(ComponentPose=@n6, Alpha=1, ExternalForce="0,0,0") {
    angularDampingOverride=0.9999
    angularSpringConstant=5000
    bAngularSpring=true
    bOverrideAngularDamping=true
    bUseGravityOverride=true
    linearSpringConstant=300
    simulationSpace=Actor
    boundBone {boneName=dyn_earring_l}
    componentAppliedLinearAccClamp=<1000 1000 1000>
    componentLinearAccScale=<0.3 0.3 0.3>
    componentLinearVelScale=<0.3 0.3 0.3>
    gravityOverride {z=-980}
    physicsBodyDefinitions [{
      boundBone {boneName=dyn_earring_l}
      boxExtents=<20 20 20>
      constraintSetup {angularLimitsMax=<0 10 5>, angularLimitsMin=<0 -10 -45>, angularTarget=<1 0 0>}
      localJointOffset=<5 0 0>
     }]
    relativeSpaceBone {boneName=DangerIndex}
   }
   n8 = Anim Dynamics - Bone: dyn_earring_r(ComponentPose=@n7, Alpha=1, ExternalForce="0,0,0") {
    angularDampingOverride=0.9999
    angularSpringConstant=5000
    bAngularSpring=true
    bOverrideAngularDamping=true
    bUseGravityOverride=true
    linearSpringConstant=300
    simulationSpace=Actor
    boundBone {boneName=dyn_earring_r}
    componentAppliedLinearAccClamp=<1000 1000 1000>
    componentLinearAccScale=<0.3 0.3 0.3>
    componentLinearVelScale=<0.3 0.3 0.3>
    gravityOverride {z=-980}
    physicsBodyDefinitions [{
      boundBone {boneName=dyn_earring_r}
      boxExtents=<20 20 20>
      constraintSetup {angularLimitsMax=<0 10 5>, angularLimitsMin=<0 -10 -45>, angularTarget=<1 0 0>}
      localJointOffset=<5 0 0>
     }]
    relativeSpaceBone {boneName=DangerIndex}
   }
   n9 = Anim Dynamics - Bone: dyn_earring_l(ComponentPose=Local To Component(LocalPose=Use cached pose 'My_pose'()), Alpha=1, ExternalForce="0,0,0") {
    angularDampingOverride=0.999
    angularSpringConstant=5000
    bAngularSpring=true
    bOverrideAngularDamping=true
    bUseGravityOverride=true
    linearSpringConstant=300
    simulationSpace=Actor
    boundBone {boneName=dyn_earring_l}
    componentAppliedLinearAccClamp=<1000 1000 1000>
    componentLinearAccScale=<0.3 0.3 0.3>
    componentLinearVelScale=<0.3 0.3 0.3>
    gravityOverride {z=-980}
    physicsBodyDefinitions [{
      boundBone {boneName=dyn_earring_l}
      boxExtents=<20 20 20>
      constraintSetup {angularLimitsMax=<0 10 5>, angularLimitsMin=<0 -10 -45>, angularTarget=<1 0 0>}
      localJointOffset=<5 0 0>
     }]
    relativeSpaceBone {boneName=DangerIndex}
   }
   n10 = Anim Dynamics - Bone: dyn_earring_r(ComponentPose=@n9, Alpha=1, ExternalForce="0,0,0") {
    angularDampingOverride=0.999
    angularSpringConstant=5000
    bAngularSpring=true
    bOverrideAngularDamping=true
    bUseGravityOverride=true
    linearSpringConstant=300
    simulationSpace=Actor
    boundBone {boneName=dyn_earring_r}
    componentAppliedLinearAccClamp=<1000 1000 1000>
    componentLinearAccScale=<0.3 0.3 0.3>
    componentLinearVelScale=<0.3 0.3 0.3>
    gravityOverride {z=-980}
    physicsBodyDefinitions [{
      boundBone {boneName=dyn_earring_r}
      boxExtents=<20 20 20>
      constraintSetup {angularLimitsMax=<0 10 5>, angularLimitsMin=<0 -10 -45>, angularTarget=<1 0 0>}
      localJointOffset=<5 0 0>
     }]
    relativeSpaceBone {boneName=DangerIndex}
   }
   n11 = Anim Dynamics - Bone: dyn_hair(ComponentPose=@n10, Alpha=1, ExternalForce="0,0,0") {
    bLinearSpring=true
    bOverrideLinearDamping=true
    bUseGravityOverride=true
    linearDampingOverride=0.9999
    linearSpringConstant=300
    simulationSpace=World
    boundBone {boneName=dyn_hair}
    componentAppliedLinearAccClamp=<1000 1000 1000>
    componentLinearAccScale=<0.7 0.7 0.7>
    componentLinearVelScale=<0.3 0.3 0.3>
    gravityOverride {z=-980}
    physicsBodyDefinitions [{
      boundBone {boneName=dyn_hair}
      boxExtents=<40 40 40>
      constraintSetup {angularLimitsMax=<0 0 0>, angularLimitsMin=<0 0 0>, angularTarget=<1 0 0>, linearAxesMax=<20 0 0>, linearAxesMin=<-20 0 0>}
      localJointOffset=<5 0 0>
     }]
   }
   n12 = Anim Dynamics - Bone: dyn_hair(ComponentPose=@n11, Alpha=1, ExternalForce="0,0,0") {
    angularDampingOverride=0.999
    angularSpringConstant=6000
    bAngularSpring=true
    bOverrideAngularDamping=true
    bUseGravityOverride=true
    linearDampingOverride=0.999
    linearSpringConstant=500
    simulationSpace=World
    boundBone {boneName=dyn_hair}
    componentAppliedLinearAccClamp=<1000 1000 1000>
    componentLinearAccScale=<0.5 0.5 0.5>
    componentLinearVelScale=<0.3 0.3 0.3>
    gravityOverride {z=-980}
    physicsBodyDefinitions [{
      boundBone {boneName=dyn_hair}
      boxExtents=<50 50 50>
      constraintSetup {angularLimitsMax=<10 10 10>, angularLimitsMin=<-10 -10 -10>, angularTarget=<1 0 0>}
      localJointOffset=<5 0 0>
     }]
   }
   n13 = RigidBody!(
    ComponentPose=@n12
    Alpha
    ExternalForce
    SimSpaceSettings=(WorldAlpha=0.8,VelocityScaleZ=1,DampingAlpha=1,MaxLinearVelocity=50,MaxAngularVelocity=1,MaxLinearAcceleration=200,MaxAngularAcceleration=4,ExternalLinearDragV=(X=0,Y=0,Z=0),ExternalLinearVelocity=(X=0,Y=0,Z=0),ExternalAngularVelocity=(X=0,Y=0,Z=0))
   ) -> (Pose) {
    bOverrideWorldGravity=true
    overridePhysicsAsset=&F_MED_DangerIndex_Pack_Physics
    simulationSpace=BaseBoneSpace
    baseBoneRef {boneName=pack}
    overrideWorldGravity {z=-980}
    simSpaceSettings {maxAngularAcceleration=4, maxAngularVelocity=1, maxLinearAcceleration=200, maxLinearVelocity=50, worldAlpha=0.8}
   }
   Output Pose(Result=Blend Poses by bool(BlendPose_0=Component To Local(ComponentPose=@n13), BlendPose_1=Component To Local(ComponentPose=@n8), bActiveValue=Get bIsInFrontEnd())) {layerGroup=None}
   Save cached pose 'My_pose'(Pose=Input Pose()) {cacheName=My_pose}
  ]}
 EventGraph {nodes [
   Event BlueprintUpdateAnimation()
   > n1 = Sequence!(execute) -> (then_0, then_1, then_2:Exec, then_3:Exec, then_4:Exec, then_5:Exec, then_6:Exec)
   Reroute Node(InputPin=@n1.then_5)
   n2 = GetOwnerSkeletalMeshComponent()
   n3 = BreakRotator(InRot=BreakTransform(InTransform=GetDeltaTransformFromRefPose(self=@n2, BoneName=spine_05, BaseName=spine_01)).Rotation)
   n4 = MapRangeClamped(Value=@n3.Yaw, InRangeA=0, InRangeB=-60, OutRangeA=1, OutRangeB=0.9)
   Set backpack_scale(execute=@n1.then_2, backpack_scale:Vector=MakeVector(X=@n4, Y=@n4, Z=@n4)) -> (Output_Get:Vector)
   n5 = "Gravity Override State Function (Parameter Asset) (Nativized)"(
    execute=@n1.then_3
    CharacterPartAnimInstance=Self-Reference()
    GameTimeInSeconds=GetGameTimeInSeconds()
    Parameters=&/F_MED_DangerIndex_Pack_Parameters
   )
   > Set Gravity Override(Gravity Override=@n5)
   > Set RBAN Settings Out(RBAN Settings Out=@n5.OutRigidBodyAnimNodeParameters)
   Set!(
    execute=@n1.then_0
    backpack_translation_X:double=MapRangeClamped(Value=@n3.Yaw, InRangeA=-15, InRangeB=-72, OutRangeB=0)
    backpack_translation_Y:double=float + float(
     A=MapRangeClamped(Value=@n3.Yaw, InRangeA=-33, InRangeB=-45, OutRangeB=-15)
     B=MapRangeClamped(Value=@n3.Yaw, InRangeA=-45, InRangeB=-72, OutRangeB=-10)
    )
    backpack_translation_Z:double=float + float(
     A=MapRangeClamped(Value=@n3.Yaw, InRangeA=-33, InRangeB=-45, OutRangeB=-10)
     B=MapRangeClamped(Value=@n3.Yaw, InRangeA=-45, InRangeB=-72, OutRangeB=8)
    )
   ) -> (then, Output_Get:Vector)
   Set!(
    execute=@n1.then_4
    pickaxe_rotate_r_Roll:float=0
    pickaxe_rotate_r_Pitch:float=0
    pickaxe_rotate_r_Yaw:float=MapRangeClamped(Value=@n3, InRangeA=20, InRangeB=50, OutRangeB=-20)
   ) -> (then, Output_Get:Rotator)
   Set!(
    execute=@n1.then_1
    backpack_rotate_Roll:float=MapRangeClamped(Value=@n3.Yaw, InRangeA=-20, InRangeB=-70, OutRangeB=60)
    backpack_rotate_Pitch:float=MapRangeClamped(Value=@n3.Yaw, InRangeA=-20, InRangeB=-70, OutRangeB=0)
    backpack_rotate_Yaw:float=MapRangeClamped(Value=@n3.Yaw, InRangeA=-20, InRangeB=-70, OutRangeB=0)
   ) -> (then, Output_Get)
   TryGetPawnOwner()
   Reroute Node(InputPin:float=@n3.Yaw) -> (OutputPin:float)
   Reroute Node(InputPin=@n3.Yaw)
   Reroute Node(InputPin=@n3.Yaw)
   Reroute Node(InputPin=@n3.Yaw)
  ]}
]
variables [
 backpack_rotate {type=Rotator}
 backpack_scale {type=Vector}
 backpack_translation {type=Vector}
 Gravity Override {type=Vector}
 NewVar {type=boolean}
 pickaxe_rotate_r {type=Rotator}
 RBAN Settings Out {type=FortRigidBodyAnimNodeParameters}
]
